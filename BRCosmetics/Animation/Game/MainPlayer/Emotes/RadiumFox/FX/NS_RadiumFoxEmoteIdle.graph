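# Niagara system for the RadiumFox emote idle, adding eye glows and a thruster flame to a skeletal mesh.
# Eyeglows pins three sprites to sockets by particle ID, fading in with system age and scaling alpha by time of day.
# Thrust follows socket 3 with layered flickering, spinning sprites and a light, using socket scale for size and camera-to-axis angle for an alpha fade.
# A HeatDistortion emitter is present but disabled.
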NiagaraSystem /BRCosmetics/Animation/Game/MainPlayer/Emotes/RadiumFox/FX/NS_RadiumFoxEmoteIdle
dependencies {
 usedDataInterfaces [&NiagaraDataInterfaceCurve, &NiagaraDataInterfaceCamera, &NiagaraDataInterfaceSkeletalMesh]
 usedDynamicInputs []
 usedModules [
  &self:SystemSpawnScript.SystemScriptSource.SystemScriptGraph.NiagaraNodeAssignment_0.SetVariables_#1
  &SystemState
 ]
}
emitters {
 Eyeglows {
  emitterUpdateScript [
   &EmitterState()
   &SpawnBurst_Instantaneous(Spawn Count=3, Use Spawn Probability=false, UseLoopCountLimit=true)
  ]
  particleSpawnScript [
   &InitializeParticle(
    Color=<rgba 0.841 0.6132 0.2944 0.25>
    Lifetime=5
    Sprite Size Mode=::NewEnumerator3
    Uniform Sprite Size=12
    UsePositionOffset=false
    Write Material Random=true
   )
   &CameraOffset(Camera Offset Amount=8)
   &DynamicMaterialParameters(
    Index 0 Param 1=3
    Index 0 Param 2=1
    Module.Index0Param0Write=true
    Module.Index0Param1Write=true
    Module.Index0Param2Write=true
    Module.Index0Param3Write=true
   )
  ]
  particleUpdateScript [
   &ParticleState(
    Kill Particles When Lifetime Has Elapsed=false
    Let Infinitely Lived Particles Die When Emitter Deactivates=true
    Loop Particles Lifetime=true
   )
   &SkeletalMeshLocation(
    Bone / Socket Index=&ReturnParticleIDAttributes()
    MaskSpawnGroup=false
    Mesh Sampling Type=::NewEnumerator5
    Skeletal Mesh=$System.Skeletal Mesh
    Socket Sampling Mode=::NewEnumerator5
   )
   &ScaleSpriteSize(
    Uniform Curve Scale=&MakeFloatFromVector(Vector=$Particles.SkeletalMeshLocation.SampledScale, Channel=::NewEnumerator0)
    Uniform Curve Sprite Scale=DataInterface() {propertyValues {Curve {keys [{tangentMode=RCTM_User, time=0, value=1}, {tangentMode=RCTM_User, time=1, value=1}]}}}
   )
   &ScaleColor(
    Scale Alpha=&FNTimeOfDayScalar(Night Scale=0.7, Morning Scale=1, Evening Scale=1)
    ScaleA=true
    ScaleRGB=true
   )
   &ScaleColor(
    Scale Alpha=&FloatFromCurve(
     FloatCurve=DataInterface() {propertyValues {Curve {keys [{tangentMode=RCTM_User, time=0, value=0}, {tangentMode=RCTM_User, time=1, value=1}]}}}
     CurveIndex=$Engine.System.Age
    )
    ScaleA=true
    ScaleRGB=true
   ) {module=ScaleColor001}
  ]
  rendererClasses [&NiagaraSpriteRendererProperties]
  renderers [{rendererClass=&NiagaraSpriteRendererProperties}]
 }
 HeatDistortion {
  bEnabled=false
  emitterUpdateScript [
   &EmitterState()
   &SpawnRate(SpawnRate=15, Use Spawn Probability=false)
  ]
  particleSpawnScript [
   &InitializeParticle(
    Color=<rgba 0 0 0 0.015>
    Lifetime=0.5
    Position Mode=::NewEnumerator0
    Sprite Rotation Mode=::NewEnumerator2
    Sprite Size Mode=::NewEnumerator3
    Uniform Sprite Size=8
    Write Material Random=true
   )
   &SkeletalMeshLocation(
    Bone / Socket Index=3
    MaskSpawnGroup=false
    Mesh Sampling Type=::NewEnumerator5
    Skeletal Mesh=$System.Skeletal Mesh
    Socket Sampling Mode=::NewEnumerator5
   )
   &AddVelocity(
    Velocity=<0 0 -50>
    Velocity Speed Scale=&MakeFloatFromVector(Vector=$Particles.SkeletalMeshLocation.SampledScale, Channel=::NewEnumerator0)
   )
   &CameraOffset(Camera Offset Amount=4)
  ]
  particleUpdateScript [
   &SolveForcesAndVelocity(Clamp Velocity=false, Limit Acceleration=false)
   &SkeletalMeshLocation(
    Bone / Socket Index=3
    MaskSpawnGroup=false
    Mesh Sampling Type=::NewEnumerator5
    Orientation Sampling=::NewEnumerator1
    Position Sampling=::NewEnumerator2
    Skeletal Mesh=$System.Skeletal Mesh
    Socket Sampling Mode=::NewEnumerator5
    Use Sampled Mesh Attribute as Ribbon ID=false
   ) {module=SkeletalMeshLocation001}
   &ParticleState()
   &ScaleSpriteSize(
    Scale Sprite Size Mode=::NewEnumerator0
    Uniform Scale Factor=&MakeFloatFromVector(Vector=$Particles.SkeletalMeshLocation.SampledScale, Channel=::NewEnumerator0)
   )
   &SpriteRotationRate(Rotation Rate=&Multiply_Float(A=90, B=&RandomRangeFloat(Minimum=-1, Maximum=1)))
   &ScaleColor(
    Scale Alpha=&FloatFromCurve(
     FloatCurve=DataInterface() {propertyValues {Curve {keys [{tangentMode=RCTM_User, time=0, value=0}, {tangentMode=RCTM_User, time=1, value=1}]}}}
     CurveIndex=$Engine.System.Age
    )
    ScaleA=true
    ScaleRGB=true
   ) {module=ScaleColor001}
  ]
  rendererClasses [&NiagaraSpriteRendererProperties]
  renderers [{rendererClass=&NiagaraSpriteRendererProperties}]
 }
 Thrust {
  emitterUpdateScript [
   &EmitterState()
   &SpawnBurst_Instantaneous(Use Spawn Probability=false, UseLoopCountLimit=true)
  ]
  particleSpawnScript [
   &InitializeParticle(
    Color=<rgba 1 0.25 0 1>
    Lifetime=5
    Sprite Rotation Angle=-90
    Sprite Rotation Mode=::NewEnumerator0
    Sprite Size=&Multiply_Vector2D(A=<8 6>, B=<1.2 1.15>)
    Sprite Size Mode=::NewEnumerator4
    Sprite UV Mode=::NewEnumerator2
    UsePositionOffset=false
    Write Material Random=true
   )
   &CameraOffset(Camera Offset Amount=3)
  ]
  particleUpdateScript [
   &SolveForcesAndVelocity(Clamp Velocity=false, Limit Acceleration=false)
   &ParticleState(
    Kill Particles When Lifetime Has Elapsed=false
    Let Infinitely Lived Particles Die When Emitter Deactivates=true
    Loop Particles Lifetime=true
   )
   &SkeletalMeshLocation(
    Bone / Socket Index=3
    MaskSpawnGroup=false
    Mesh Sampling Type=::NewEnumerator5
    Sampled Position Offset=<-1.9423 0 0>
    Skeletal Mesh=$System.Skeletal Mesh
    Socket Sampling Mode=::NewEnumerator5
   )
   &SpriteFacingAndAlignment(
    Sprite Alignment=&Multiply_VectorByFloat(
     Vector=&ReturnMeshOrientationAxis(Mesh Orientation Axis=::XAxis, Mesh Orientation Quat=$Particles.SkeletalMeshLocation.SampledRotation)
     Float=1
    )
    Sprite Facing=<0 0 -1>
    WriteAlignment=true
    WriteFacing=false
   )
   /SetVariables(Particles.AnimScale:NiagaraFloat=&MakeFloatFromVector(Vector=$Particles.SkeletalMeshLocation.SampledScale, Channel=::NewEnumerator0)) {isSetParameters=true}
   /SetVariables(
    Particles.Custom Alighnment:Vector3f=<0 1 0>
    Particles.Custom Color:LinearColor=<rgba 1.5 0.75 0 1>
    Particles.Custom Facing:Vector3f=<0 0 -1>
    Particles.Custom Position:NiagaraPosition=&AddVectorToPosition(
     Position=$Particles.Position
     Vector=&Multiply_Vector(
      A=&ReturnMeshOrientationAxis(Mesh Orientation Axis=::XAxis, Mesh Orientation Quat=$Particles.SkeletalMeshLocation.SampledRotation)
      B=&VectorFromFloat(Value=1.9329)
     )
    )
    Particles.Custom Rotation:NiagaraFloat=&Multiply_Float(A=$Emitter.Age, B=-690.1524)
    Particles.Custom Rotation 2:NiagaraFloat=&Multiply_Float(A=$Emitter.Age, B=560.9257)
    Particles.Custom Size:Vector2f=&Multiply_Vector2D_ByFloat(
     Float=&Multiply_Float(A=10, B=$Particles.AnimScale)
     Vector2D=&MakeVector2D(
      X=&Waveform(
       Global Input Position=&RandomRangeFloat(Minimum=0, Maximum=100000)
       "Waveform [1]"=::NewEnumerator12
       "[1] Frequency"=10
       "[1] Amplitude Min/Max"=<0.8 1>
      )
      Y=&Waveform(
       Global Input Position=&RandomRangeFloat(Minimum=0, Maximum=100000)
       "Waveform [1]"=::NewEnumerator12
       "[1] Frequency"=10
       "[1] Amplitude Min/Max"=<0.8 1>
      )
     )
    )
    Particles.Custom Size 2:Vector2f=&Multiply_Vector2D_ByFloat(
     Float=&Multiply_Float(A=5, B=$Particles.AnimScale)
     Vector2D=&MakeVector2D(
      X=&Waveform(
       Global Input Position=&RandomRangeFloat(Minimum=0, Maximum=100000)
       "Waveform [1]"=::NewEnumerator12
       "[1] Frequency"=10
       "[1] Amplitude Min/Max"=<0.8 1>
      )
      Y=&Waveform(
       Global Input Position=&RandomRangeFloat(Minimum=0, Maximum=100000)
       "Waveform [1]"=::NewEnumerator12
       "[1] Frequency"=10
       "[1] Amplitude Min/Max"=<0.8 1>
      )
     )
    )
   ) {isSetParameters=true}
   &ScaleSpriteSize(Scale Sprite Size Mode=::NewEnumerator0, Uniform Scale Factor=$Particles.AnimScale)
   /SetVariables(Particles.Flame COlor:LinearColor=&MakeLinearColor(
     R:NiagaraFloat=1
     G:NiagaraFloat=1
     B:NiagaraFloat=1
     A:NiagaraFloat=&FloatFromCurve(
      FloatCurve=DataInterface() {propertyValues {Curve {keys [
          {arriveTangentWeight=0.2668, interpMode=RCIM_Cubic, leaveTangentWeight=0.1601, tangentMode=RCTM_User, time=0.9, value=1}
          {arriveTangentWeight=0.1, interpMode=RCIM_Cubic, tangentMode=RCTM_User, time=1, value=0}
         ]}}}
      CurveIndex=&Abs_Float(Float:NiagaraFloat=&DotProduct(
        Vector A=&ReturnCameraVector(Camera Vectors=::NewEnumerator4, Coordinate Space=::Simulation, Origin Position=$Particles.Position)
        Vector B=&ReturnMeshOrientationAxis(Mesh Orientation Axis=::XAxis, Mesh Orientation Quat=$Particles.SkeletalMeshLocation.SampledRotation)
       ))
     )
    )) {isSetParameters=true}
   /SetVariables(Particles.FlamePosition:NiagaraPosition=&LerpPosition(
     A=$Particles.Position
     B=&AddVectorToPosition(Position=$Particles.Position, Vector=<0 0 0>)
     Alpha=&MakeFloatFromVector(Vector=<0 0 0>, Channel=::NewEnumerator0)
    )) {enabled=false, isSetParameters=true}
   &FN_Light_Attributes(
    Color=<rgba 1.25 0.3125 0 1>
    Create Color Attribute=true
    Create Exponent Attribute=true
    Create Position Attribute=true
    Create Radius Attribute=true
    Exponent=2
    Position=&AddVectorToPosition(
     Position=$Particles.Position
     Vector=&Multiply_VectorByFloat(
      Vector=&ReturnMeshOrientationAxis(Mesh Orientation Axis=::XAxis, Mesh Orientation Quat=$Particles.MeshOrientation)
      Float=8
     )
    )
    Radius=&Multiply_Float(A=15, B=$Particles.AnimScale)
    Scale Color=&Multiply_Float(
     A=&Waveform("Waveform [1]"=::NewEnumerator12, "[1] Frequency"=15, "[1] Amplitude Min/Max"=<4 2>)
     B=$Particles.AnimScale
    )
   )
  ]
  rendererClasses [&NiagaraSpriteRendererProperties, &NiagaraLightRendererProperties]
  renderers [
   {rendererClass=&NiagaraSpriteRendererProperties}
   {rendererClass=&NiagaraSpriteRendererProperties, rendererIndex=1}
   {rendererClass=&NiagaraSpriteRendererProperties, rendererIndex=2}
   {rendererClass=&NiagaraSpriteRendererProperties, rendererIndex=3}
   {rendererClass=&NiagaraSpriteRendererProperties, rendererIndex=4}
   {rendererClass=&NiagaraLightRendererProperties, rendererIndex=5}
  ]
 }
}
summary {
 systemName=NS_RadiumFoxEmoteIdle
 emitters [
  {emitterName=Eyeglows, rendererClasses [&NiagaraSpriteRendererProperties]}
  {emitterName=Thrust, rendererClasses [&NiagaraSpriteRendererProperties, &NiagaraLightRendererProperties]}
  {bEnabled=false, emitterName=HeatDistortion, rendererClasses [&NiagaraSpriteRendererProperties]}
 ]
 userVariables []
}
